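# Translucent UI master material for featured wrap images in the offer catalog, compositing a wrap icon over an item render (SCAR weapon icon by default).
# Both textures are parameters with their own center scale, offset and rotation controls.
# The icon's alpha blends RGB and combines coverage through lerps.
# A final lerp places the composite over constant red, but the graph shows no connection to material outputs.
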Material /OfferCatalog/Textures/Icons/Wraps/Materials/M-FeatureWrap-MasterComposit
graph {
 properties {
  bUsedWithStaticMesh=false
  blendMode=BLEND_Translucent
  materialDomain=MD_UI
  pixelDepthOffsetMode=PDOM_Legacy
  usedShadingModels=Default Lit
 }
 nodes [
  n1 = MaterialFunctionCall(
   UVs=TextureCoordinate()
   Texture Scale=ScalarParameter() {defaultValue=0.8024, group=Item Render, parameterName=UV-Scale-Factor, sliderMax=1, sliderMin=0.5}
  ) {materialFunction=&ScaleUVsByCenter}
  n2 = AppendVector(
   A=ScalarParameter() {defaultValue=-0.0095, group=Item Render, parameterName=UV-XOffset}
   B=ScalarParameter() {defaultValue=0.0964, group=Item Render, parameterName=UV-YOffset}
  )
  n3 = TextureSampleParameter2D(UVs=Rotator(Coordinate=Add(A=@n1.UVs, B=@n2), Time=ScalarParameter() {defaultValue=-0.003, group=Item Render, parameterName=Rotation}) {speed=6.2832}) {
   group=Item Render
   parameterName=Item Render
   samplerSource=SSM_Clamp_WorldGroupSettings
   texture=/BRCosmetics/UI/Foundation/Textures/Icons/Weapons/Items/T-Icon-Weapons-SK-SCAR-L
  }
  n4 = MaterialFunctionCall(
   UVs=TextureCoordinate()
   Texture Scale=ScalarParameter() {defaultValue=0.25, group=Wrap Icon, parameterName=Icon UV-Scale-Factor, sliderMax=1, sliderMin=0.5}
  ) {materialFunction=&ScaleUVsByCenter}
  n5 = AppendVector(
   A=ScalarParameter() {defaultValue=-0.8452, group=Wrap Icon, parameterName=Icon-UV-XOffset}
   B=ScalarParameter() {defaultValue=-0.3667, group=Wrap Icon, parameterName=Icon-UV-YOffset}
  )
  n6 = TextureSampleParameter2D(UVs=Rotator(Coordinate=Add(A=@n4.UVs, B=@n5), Time=ScalarParameter() {defaultValue=0.0405, group=Wrap Icon, parameterName=Icon-Rotation}) {speed=6.2832}) {
   group=Wrap Icon
   parameterName=Wrap Icon
   samplerSource=SSM_Clamp_WorldGroupSettings
   texture=/BRCosmetics/UI/Foundation/Textures/Icons/Wraps/T-Wraps-Wrap-NeoTag
  }
  LinearInterpolate(
   A=Constant3Vector() {constant=<rgba 0.9844 0.0051 0.0051 0>}
   B=LinearInterpolate(A=@n3.RGB, B=@n6.RGB, Alpha=@n6.A) {desc=RGB Composit}
   Alpha=LinearInterpolate(A=@n3.A, Alpha=@n6.A) {desc=Alpha Composit}
  )
 ]
}
