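# UI material for the StreetDemon wrap icon in the offer catalog, adding an animated colored glow over the base texture.
# A distortion texture supplies coordinates that pan on a looping Fmod timer, forming a moving gradient without a gradient texture.
# The gradient thresholds and blends the mask's red and green channels, with blue gating where the tint is added.
# Base alpha passes through, and parameters scale and offset the UVs.
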Material /OfferCatalog/Textures/Icons/Wraps/Materials/M-Wraps-Animated-StreetDemon
graph {
 properties {
  bUsedWithStaticMesh=false
  blendMode=BLEND_Translucent
  materialDomain=MD_UI
  numCustomizedUVs=1
  pixelDepthOffsetMode=PDOM_Legacy
  usedShadingModels=Default Lit
 }
 nodes [
  Add(
   A=MaterialFunctionCall(UVs=TextureCoordinate(), Texture Scale=ScalarParameter() {defaultValue=1, parameterName=UV-Scale-Factor, sliderMax=1, sliderMin=0.5}) {materialFunction=&ScaleUVsByCenter}.UVs
   B=AppendVector(A=ScalarParameter() {parameterName=UV-XOffset}, B=ScalarParameter() {parameterName=UV-YOffset})
  )
  n1 = TextureSampleParameter2D() {
   parameterName=Base-Texture
   samplerSource=SSM_Clamp_WorldGroupSettings
   texture=&/T-StreetDemon-Difuse
  }
  n2 = Multiply(A=@n1.RGB, B=ScalarParameter() {defaultValue=1, parameterName=ColorMultiply, sliderMax=1})
  n3 = TextureSampleParameter2D() {parameterName=Mask, texture=&/T-StreetDemon-M}
  n4 = TextureSampleParameter2D() {parameterName=Distortion-UV, samplerSource=SSM_Clamp_WorldGroupSettings, texture=&/T-Wraps-AnimUV}
  n5 = ComponentMask(None=@n4.RGB) {g=true, r=true}
  n6 = Time()
  n7 = ScalarParameter() {defaultValue=3.25, parameterName=FMod}
  n8 = ComponentMask(None=Panner(Coordinate=Add(A=@n5) {constA=-0.5, constB=0.5}, Time=Fmod(A=@n6, B=@n7), Speed=ScalarParameter() {defaultValue=-0.35, parameterName=Speed}) {speedY=-0.5}) {g=true}
  n9 = ScalarParameter() {parameterName=Gradient Edge}
  n10 = Saturate(None=Multiply(A=Subtract(A=@n8, B=@n9), B=ScalarParameter() {defaultValue=1.8, parameterName=Gradient Multi}) {constB=4})
  n11 = Subtract(A=@n3.R, B=LinearInterpolate(Alpha=@n10) {constA=1, constB=0})
  n12 = ScalarParameter() {defaultValue=0.125, parameterName=Pass-2-edge}
  n13 = Saturate(None=Multiply(A=Subtract(A=@n11, B=@n12), B=ScalarParameter() {defaultValue=4, parameterName=Pass-2-Multi}) {constB=4})
  n14 = Subtract(A=@n3.G, B=@n10)
  n15 = ScalarParameter() {defaultValue=0.25, parameterName=Pass-1-edge}
  n16 = Saturate(None=Multiply(A=Subtract(A=@n14, B=@n15), B=ScalarParameter() {defaultValue=4, parameterName=Pass-1-Multi}) {constB=4})
  n17 = Clamp(None=LinearInterpolate(B=LinearInterpolate(A=@n13, B=@n16, Alpha=@n10), Alpha=@n3.B))
  n18 = Multiply(A=@n17, B=ScalarParameter() {defaultValue=1.5, parameterName=Color-Multipli}) {constB=2}
  n19 = VectorParameter() {parameterName=Color, defaultValue=<rgba 0.849 0.1282 0.5781 1>}
  LinearInterpolate(A=@n2, B=Add(A=@n2, B=Multiply(A=@n18, B=@n19.RGB)), Alpha=@n17)
  Reroute(None=@n1.A)
 ]
}
